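# Sound cue for evening and night cicada spot ambience in the Olympus biome, classed as an outdoor ambient one-shot.
# A quality switch feeds the first two levels a weighted random pick favoring the 7.5-second clip over the 10-second one.
# The last two levels play only a shorter 5-second clip.
# It uses bird one-shot attenuation and a distance-based send to the exterior ambient source bus.
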SoundCue /Game/Sounds/Helios/Biomes/Olympus/Spots/Spot_Olympus_Cicadas_Evening_Night_Cue
root=QualityLevel_0
nodes [
 {id=QualityLevel_0, children [Random_1,Random_1,WavePlayer_3,WavePlayer_3]}
 {id=Random_1, children [WavePlayer_0,WavePlayer_2], properties {weights [1,0.5]}}
 {id=WavePlayer_0, properties {soundWaveAssetPtr=&/Spot_Olympus_Cicadas_02_Evening_Night_7500ms}}
 {id=WavePlayer_2, properties {soundWaveAssetPtr=&/Spot_Olympus_Cicadas_02_Evening_Night_10000ms}}
 {id=WavePlayer_3, properties {soundWaveAssetPtr=&/Spot_Olympus_Cicadas_02_Evening_Night_5000ms}}
]
properties {
 attenuationSettings=&Fort_TileAmbient_OneShot_Bird_Att
 maxDistance=8005
 soundClassObject=&Fort_Ambient_OneShots_Outdoors
 volumeMultiplier=0.501
 busSends [{
   audioBus=null
   maxSendDistance=6000
   minSendDistance=3000
   minSendLevel=0.6
   sendLevel=1
   soundSourceBus=&AmbientAudio_Exterior_SourceBus
  }]
}
