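# Sound cue for daytime wind one-shot spot ambience in the desolate Olympus biome, on the outdoor ambient class with tile one-shot attenuation.
# A quality switch sends its first two levels to a random choice between tonal Olympus gusts and reused Alpine quad one-shots.
# The last two levels play only Alpine ST one-shots, at slightly lower volume.
# Each pool picks a random wave, varied by pitch and volume modulation.
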SoundCue /Game/Sounds/Helios/Biomes/Olympus/Spots/Spot_Olympus_Desolate_Wind_Oneshots_Day_Cue
root=QualityLevel_0
nodes [
 {id=Modulator_0, children [Random_1], properties {pitchMax=1.1, pitchMin=0.9, volumeMax=0.9, volumeMin=0.8}}
 {id=Modulator_1, children [Random_0], properties {pitchMax=1.1, pitchMin=0.9, volumeMax=1.1, volumeMin=0.9}}
 {id=Modulator_2, children [Random_4], properties {pitchMax=1.1, pitchMin=0.9, volumeMax=1.1, volumeMin=0.9}}
 {id=QualityLevel_0, children [Random_3,Random_3,Modulator_0,Modulator_0]}
 {id=Random_0, children [WavePlayer_1,WavePlayer_2,WavePlayer_3,WavePlayer_4], properties {weights [1,1,1,1]}}
 {id=Random_1, children [WavePlayer_5,WavePlayer_6,WavePlayer_7,WavePlayer_8], properties {weights [1,1,1,1]}}
 {id=Random_3, children [Modulator_2,Modulator_1], properties {weights [1,1]}}
 {
  id=Random_4
  children [WavePlayer_19, WavePlayer_20, WavePlayer_21, WavePlayer_22, WavePlayer_23, WavePlayer_24, WavePlayer_25, WavePlayer_26]
  properties {weights [1,1,1,1,1,1,1,1]}
 }
 {id=WavePlayer_1, properties {soundWaveAssetPtr=/Game/Sounds/Apollo/Biomes/Alpine/Biome_Alpine_Oneshot_Day_Quad_01}}
 {id=WavePlayer_19, properties {soundWaveAssetPtr=&/Spot_Olympus_Wind_Gust_Tonal_1}}
 {id=WavePlayer_2, properties {soundWaveAssetPtr=/Game/Sounds/Apollo/Biomes/Alpine/Biome_Alpine_Oneshot_Day_Quad_02}}
 {id=WavePlayer_20, properties {soundWaveAssetPtr=&/Spot_Olympus_Wind_Gust_Tonal_2}}
 {id=WavePlayer_21, properties {soundWaveAssetPtr=&/Spot_Olympus_Wind_Gust_Tonal_3}}
 {id=WavePlayer_22, properties {soundWaveAssetPtr=&/Spot_Olympus_Wind_Gust_Tonal_4}}
 {id=WavePlayer_23, properties {soundWaveAssetPtr=&/Spot_Olympus_Wind_Gust_Tonal_5}}
 {id=WavePlayer_24, properties {soundWaveAssetPtr=&/Spot_Olympus_Wind_Gust_Tonal_6}}
 {id=WavePlayer_25, properties {soundWaveAssetPtr=&/Spot_Olympus_Wind_Gust_Tonal_7}}
 {id=WavePlayer_26, properties {soundWaveAssetPtr=&/Spot_Olympus_Wind_Gust_Tonal_8}}
 {id=WavePlayer_3, properties {soundWaveAssetPtr=/Game/Sounds/Apollo/Biomes/Alpine/Biome_Alpine_Oneshot_Day_Quad_03}}
 {id=WavePlayer_4, properties {soundWaveAssetPtr=/Game/Sounds/Apollo/Biomes/Alpine/Biome_Alpine_Oneshot_Day_Quad_04}}
 {id=WavePlayer_5, properties {soundWaveAssetPtr=/Game/Sounds/Apollo/Biomes/Alpine/Biome_Alpine_Oneshot_Day_ST_01}}
 {id=WavePlayer_6, properties {soundWaveAssetPtr=/Game/Sounds/Apollo/Biomes/Alpine/Biome_Alpine_Oneshot_Day_ST_02}}
 {id=WavePlayer_7, properties {soundWaveAssetPtr=/Game/Sounds/Apollo/Biomes/Alpine/Biome_Alpine_Oneshot_Day_ST_03}}
 {id=WavePlayer_8, properties {soundWaveAssetPtr=/Game/Sounds/Apollo/Biomes/Alpine/Biome_Alpine_Oneshot_Day_ST_04}}
]
properties {
 attenuationSettings=/Game/Sounds/Attenuation/Fort_TileAmbient_OneShot_Att
 maxDistance=8000
 soundClassObject=&Fort_Ambient_Outdoors_Athena
 volumeMultiplier=0.15
 preEffectBusSends [{}]
}
